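# Sound cue for the quest outfit transformation effect, played through the non-critical in-game UI sound class.
# A single wave player plays one transformation sound at slightly boosted volume, using the quest interactable hint attenuation.
# A concurrency override allows only one instance at a time, stopping the oldest when a new one starts.
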SoundCue /FNGameplayCues/Athena/Sounds/Quests/SkinTransition/Quest_Outfit_Transformation_Cue
root=WavePlayer_1
nodes [{id=WavePlayer_1, properties {soundWaveAssetPtr=&/Quest_Outfit_Transformation_01}}]
properties {
 attenuationSettings=/Game/Sounds/Attenuation/Fort_QuestInteractables_Hint
 bOverrideConcurrency=true
 maxDistance=1300
 soundClassObject=&Fort_UI_InGame_NonCritical
 volumeMultiplier=1.2
 concurrencyOverrides {maxCount=1, resolutionRule=StopOldest}
}
